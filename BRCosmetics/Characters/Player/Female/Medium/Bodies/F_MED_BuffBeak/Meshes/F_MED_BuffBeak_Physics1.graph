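# Physics asset for the F_MED_BuffBeak female medium player body cosmetic, set up around a dangling scarf.
# Four scarf capsules form a constrained chain from a small scarf_MSTR sphere, with tight twist limits and swing limits that widen toward the tip.
# Unconstrained capsules on the spine and right arm only give body volumes around the scarf, likely as colliders.
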PhysicsAsset /BRCosmetics/Characters/Player/Female/Medium/Bodies/F_MED_BuffBeak/Meshes/F_MED_BuffBeak_Physics1
bodies {
 dyn_scarf_01 [{shapeType=Capsule, length=5.5, radius=3, shapeName=dyn_scarf_01_capsule, center=<3.1271 0 0>, rotation=<rot 90 179.9851 179.9852>}]
 dyn_scarf_02 [{shapeType=Capsule, length=5.5, radius=3, shapeName=dyn_scarf_02_capsule, center=<3.1271 0 -0.0162>, rotation=<rot 90 179.9472 179.9473>}]
 dyn_scarf_03 [{shapeType=Capsule, length=4.5, radius=2, shapeName=dyn_scarf_03_capsule, center=<4.1786 0 -0.2638>, rotation=<rot 90 179.8417 179.8419>}]
 dyn_scarf_04 [{shapeType=Capsule, length=4.5, radius=4, shapeName=dyn_scarf_04_capsule, center=<3.1271 0 2.0125>, rotation=<rot 90 0 0.0002>}]
 lowerarm_r [{shapeType=Capsule, length=27.5, radius=10.5, shapeName=lowerarm_r_capsule, center=<-23.2542 10.3025 -5.1597>, rotation=<rot 90 0 -0.0004>}]
 scarf_MSTR [{shapeType=Sphere, radius=1, shapeName=scarf_MSTR_sphere}]
 spine_02 [{shapeType=Capsule, length=10, radius=20, shapeName=spine_02_capsule, center=<0 1.149 0>}]
 spine_03 [{shapeType=Capsule, length=15, radius=15, shapeName=spine_03_capsule, center=<0 4.5335 0>}]
 spine_04 [{shapeType=Capsule, length=14, radius=16, shapeName=spine_04_capsule, center=<0 -1.2802 0>}]
 upperarm_r [{shapeType=Capsule, length=26, radius=4, shapeName=upperarm_r_capsule, center=<-15.6851 -0.0001 0>, rotation=<rot 90 -160.5289 -160.5294>}]
}
constraints [
 {bone1Name=dyn_scarf_02, bone2Name=dyn_scarf_01, swing1LimitDegrees=15, swing2LimitDegrees=10, twistLimitDegrees=1}
 {bone1Name=dyn_scarf_03, bone2Name=dyn_scarf_02, swing1LimitDegrees=25, swing2LimitDegrees=15, twistLimitDegrees=1}
 {bone1Name=dyn_scarf_04, bone2Name=dyn_scarf_03, swing1LimitDegrees=25, swing2LimitDegrees=15, twistLimitDegrees=1}
 {bone1Name=dyn_scarf_01, bone2Name=scarf_MSTR, swing1LimitDegrees=15, swing2LimitDegrees=10, twistLimitDegrees=1}
]
