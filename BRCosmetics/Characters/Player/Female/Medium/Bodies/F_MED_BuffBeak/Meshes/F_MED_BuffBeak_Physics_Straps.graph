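# Physics asset for the straps on the female medium BuffBeak player cosmetic body, defining simulated secondary motion for left and right strap bone chains.
# Each side has a small sphere on its base bone and larger offset spheres on three chained segments.
# Constraints link each segment to the one before it, with almost no twist and a wider first swing limit at the root joint than further down.
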PhysicsAsset /BRCosmetics/Characters/Player/Female/Medium/Bodies/F_MED_BuffBeak/Meshes/F_MED_BuffBeak_Physics_Straps
bodies {
 dyn_strap_l_01 [{shapeType=Sphere, radius=3.5, shapeName=dyn_strap_l_01_sphere, center=<3.7818 0 0>}]
 dyn_strap_l_02 [{shapeType=Sphere, radius=3.5, shapeName=dyn_strap_l_02_sphere, center=<3.7818 0 0>}]
 dyn_strap_l_03 [{shapeType=Sphere, radius=3.5, shapeName=dyn_strap_l_03_sphere, center=<3.7818 0 0>}]
 dyn_strap_l_base [{shapeType=Sphere, radius=1, shapeName=dyn_strap_l_base_sphere}]
 dyn_strap_r_01 [{shapeType=Sphere, radius=3.5, shapeName=dyn_strap_r_01_sphere, center=<3.7818 0 0>}]
 dyn_strap_r_02 [{shapeType=Sphere, radius=3.5, shapeName=dyn_strap_r_02_sphere, center=<3.7818 0 0>}]
 dyn_strap_r_03 [{shapeType=Sphere, radius=3.5, shapeName=dyn_strap_r_03_sphere, center=<3.7818 0 0>}]
 dyn_strap_r_base [{shapeType=Sphere, radius=1, shapeName=dyn_strap_r_base_sphere}]
}
constraints [
 {bone1Name=dyn_strap_l_01, bone2Name=dyn_strap_l_base, swing1LimitDegrees=40, swing2LimitDegrees=20, twistLimitDegrees=2}
 {bone1Name=dyn_strap_l_02, bone2Name=dyn_strap_l_01, swing1LimitDegrees=25, swing2LimitDegrees=20, twistLimitDegrees=2}
 {bone1Name=dyn_strap_l_03, bone2Name=dyn_strap_l_02, swing1LimitDegrees=25, swing2LimitDegrees=20, twistLimitDegrees=2}
 {bone1Name=dyn_strap_r_01, bone2Name=dyn_strap_r_base, swing1LimitDegrees=40, swing2LimitDegrees=20, twistLimitDegrees=2}
 {bone1Name=dyn_strap_r_02, bone2Name=dyn_strap_r_01, swing1LimitDegrees=25, swing2LimitDegrees=20, twistLimitDegrees=2}
 {bone1Name=dyn_strap_r_03, bone2Name=dyn_strap_r_02, swing1LimitDegrees=25, swing2LimitDegrees=20, twistLimitDegrees=2}
]
