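# Post-process style Anim Blueprint for the F_MED_BuffBeak cosmetic body, adding secondary motion to straps and scarf.
# It copies the parent mesh pose, runs a Control Rig fed an emote-or-skydive flag, modifies right-arm bones, and caches the result.
# bRigidBodyEnabled selects two chained RigidBody sims or an AnimDynamics fallback chain; the microphone node is unconnected.
# The EventGraph updates gravity override and records BodyMesh right-arm transforms.
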AnimBlueprint /BRCosmetics/Characters/Player/Female/Medium/Bodies/F_MED_BuffBeak/Meshes/F_MED_BuffBeak_AnimBP
target_skeleton=&/F_MED_BuffBeak_Skeleton
graphs [
 AnimGraph {nodes [
   Anim Dynamics - Bone: dyn_microphone(Alpha=1, ExternalForce="0,0,0") {
    bUseGravityOverride=true
    simulationSpace=BoneRelative
    alphaBoolBlend {blendOption=Linear}
    boundBone {boneName=dyn_microphone}
    componentAppliedLinearAccClamp=<10 10 10>
    componentLinearAccScale=<0.05 0.05 0.05>
    componentLinearVelScale=<0.05 0.05 0.05>
    gravityOverride {z=-300}
    physicsBodyDefinitions [{
      boundBone {boneName=dyn_microphone}
      boxExtents=<10 15 15>
      constraintSetup {angularLimitsMax=<0 0 5>, angularLimitsMin=<0 0 -5>, angularTarget=<0 0 0>}
      localJointOffset=<1 0 0>
     }]
    relativeSpaceBone {boneName=neck_02}
   }
   n1 = RigidBody!(ComponentPose=Local To Component(LocalPose=Use cached pose 'PreDynamics'()), Alpha, OverrideWorldGravity:Vector="0,0,-980", ExternalForce) -> (Pose) {
    bOverrideWorldGravity=true
    overridePhysicsAsset=&/F_MED_BuffBeak_Physics1
    simulationSpace=BaseBoneSpace
    alphaBoolBlend {blendOption=Linear}
    baseBoneRef {boneName=spine_01}
    componentAppliedLinearAccClamp=<3000 3000 3000>
    componentLinearAccScale=<0.25 0.2 0.2>
    componentLinearVelScale=<0.25 0.2 0.2>
    overrideWorldGravity {z=-980}
   }
   n2 = RigidBody(ComponentPose=@n1, OverrideWorldGravity="0,0,-980") {
    bOverrideWorldGravity=true
    overridePhysicsAsset=&/F_MED_BuffBeak_Physics_Straps
    simulationSpace=BaseBoneSpace
    alphaBoolBlend {blendOption=Linear}
    baseBoneRef {boneName=neck_02}
    componentAppliedLinearAccClamp=<3200 3200 3200>
    componentLinearAccScale=<0.25 0.18 0.3>
    componentLinearVelScale=<0.25 0.18 0.3>
    overrideWorldGravity {z=-980}
   }
   n3 = Anim Dynamics - Bone: dyn_strap_r_01(ComponentPose=Local To Component(LocalPose=Use cached pose 'PreDynamics'()), Alpha=1, ExternalForce="0,0,0") {
    angularDampingOverride=0.999
    bOverrideAngularDamping=true
    bUseGravityOverride=true
    alphaBoolBlend {blendOption=Linear}
    boundBone {boneName=dyn_strap_r_01}
    componentAppliedLinearAccClamp=<5000 5000 5000>
    componentLinearAccScale=<0.25 0.25 0.25>
    componentLinearVelScale=<0.25 0.25 0.25>
    gravityOverride {z=-700}
    physicsBodyDefinitions [{
      boundBone {boneName=dyn_strap_r_01}
      boxExtents=<10 5 5>
      constraintSetup {angularLimitsMax=<0 5 50>, angularLimitsMin=<0 -45 -50>, angularTarget=<0 0 0>}
      localJointOffset=<1.5 0 0>
     }]
    relativeSpaceBone {boneName=spine_05}
   }
   n4 = Anim Dynamics - Bone: dyn_strap_r_02(ComponentPose=@n3, Alpha=1, ExternalForce="0,0,0") {
    angularDampingOverride=0.999
    bOverrideAngularDamping=true
    bUseGravityOverride=true
    alphaBoolBlend {blendOption=Linear}
    boundBone {boneName=dyn_strap_r_02}
    componentAppliedLinearAccClamp=<4000 4000 4000>
    componentLinearAccScale=<0.25 0.25 0.25>
    componentLinearVelScale=<0.25 0.25 0.25>
    gravityOverride {z=-700}
    physicsBodyDefinitions [{
      boundBone {boneName=dyn_strap_r_02}
      boxExtents=<7 5 5>
      constraintSetup {angularLimitsMax=<0 0 60>, angularLimitsMin=<0 -45 -60>, angularTarget=<0 0 0>}
      localJointOffset=<1 0 0>
     }]
    relativeSpaceBone {boneName=spine_05}
   }
   n5 = Anim Dynamics - Bone: dyn_strap_l_01(ComponentPose=@n4, Alpha=1, ExternalForce="0,0,0") {
    angularDampingOverride=0.999
    bOverrideAngularDamping=true
    bUseGravityOverride=true
    alphaBoolBlend {blendOption=Linear}
    boundBone {boneName=dyn_strap_l_01}
    componentAppliedLinearAccClamp=<5000 5000 5000>
    componentLinearAccScale=<0.25 0.25 0.25>
    componentLinearVelScale=<0.25 0.25 0.25>
    gravityOverride {z=-700}
    physicsBodyDefinitions [{
      boundBone {boneName=dyn_strap_l_01}
      boxExtents=<10 5 5>
      constraintSetup {angularLimitsMax=<0 45 50>, angularLimitsMin=<0 -5 -50>, angularTarget=<0 0 0>}
      localJointOffset=<1.5 0 0>
     }]
    relativeSpaceBone {boneName=spine_05}
   }
   n6 = Anim Dynamics - Bone: dyn_strap_l_02(ComponentPose=@n5, Alpha=1, ExternalForce="0,0,0") {
    angularDampingOverride=0.999
    bOverrideAngularDamping=true
    bUseGravityOverride=true
    alphaBoolBlend {blendOption=Linear}
    boundBone {boneName=dyn_strap_l_02}
    componentAppliedLinearAccClamp=<4000 4000 4000>
    componentLinearAccScale=<0.25 0.25 0.25>
    componentLinearVelScale=<0.25 0.25 0.25>
    gravityOverride {z=-700}
    physicsBodyDefinitions [{
      boundBone {boneName=dyn_strap_l_02}
      boxExtents=<7 5 5>
      constraintSetup {angularLimitsMax=<0 45 60>, angularLimitsMin=<0 0 -60>, angularTarget=<0 0 0>}
      localJointOffset=<1 0 0>
     }]
    relativeSpaceBone {boneName=spine_05}
   }
   n7 = Anim Dynamics - Bone: dyn_scarf_01(ComponentPose=@n6, Alpha=1, ExternalForce="0,0,0") {
    angularDampingOverride=0.9999
    bOverrideAngularDamping=true
    bUseGravityOverride=true
    simulationSpace=RootRelative
    alphaBoolBlend {blendOption=Linear}
    boundBone {boneName=dyn_scarf_01}
    componentAppliedLinearAccClamp=<500 500 500>
    componentLinearAccScale=<0.2 0.2 0.2>
    componentLinearVelScale=<0.2 0.2 0.2>
    gravityOverride {z=-900}
    physicsBodyDefinitions [{
      boundBone {boneName=dyn_scarf_01}
      boxExtents=<15 7 8>
      constraintSetup {angularLimitsMax=<0 0 30>, angularLimitsMin=<0 -45 -30>, angularTarget=<0 0 0>}
      localJointOffset=<1.3 0 0>
     }]
   }
   n8 = Anim Dynamics - Bone: dyn_scarf_02(ComponentPose=@n7, Alpha=1, ExternalForce="0,0,0") {
    angularDampingOverride=0.9999
    bOverrideAngularDamping=true
    bUseGravityOverride=true
    simulationSpace=RootRelative
    alphaBoolBlend {blendOption=Linear}
    boundBone {boneName=dyn_scarf_02}
    componentAppliedLinearAccClamp=<500 500 500>
    componentLinearAccScale=<0.2 0.2 0.2>
    componentLinearVelScale=<0.2 0.2 0.2>
    gravityOverride {z=-900}
    physicsBodyDefinitions [{
      boundBone {boneName=dyn_scarf_02}
      boxExtents=<15 7 8>
      constraintSetup {angularLimitsMax=<0 0 30>, angularLimitsMin=<0 -45 -30>, angularTarget=<0 0 0>}
      localJointOffset=<1.3 0 0>
     }]
   }
   n9 = Anim Dynamics - Bone: dyn_scarf_03(ComponentPose=@n8, Alpha=1, ExternalForce="0,0,0") {
    angularDampingOverride=0.9999
    bOverrideAngularDamping=true
    bUseGravityOverride=true
    simulationSpace=RootRelative
    alphaBoolBlend {blendOption=Linear}
    boundBone {boneName=dyn_scarf_03}
    componentAppliedLinearAccClamp=<500 500 500>
    componentLinearAccScale=<0.2 0.2 0.2>
    componentLinearVelScale=<0.2 0.2 0.2>
    gravityOverride {z=-900}
    physicsBodyDefinitions [{
      boundBone {boneName=dyn_scarf_03}
      boxExtents=<13 10 8>
      constraintSetup {angularLimitsMax=<0 0 30>, angularLimitsMin=<0 -45 -30>, angularTarget=<0 0 0>}
      localJointOffset=<1.5 0 0>
     }]
   }
   n10 = Anim Dynamics - Bone: dyn_scarf_04(ComponentPose=@n9, Alpha=1, ExternalForce="0,0,0") {
    angularDampingOverride=0.9999
    bOverrideAngularDamping=true
    bUseGravityOverride=true
    simulationSpace=RootRelative
    alphaBoolBlend {blendOption=Linear}
    boundBone {boneName=dyn_scarf_04}
    componentAppliedLinearAccClamp=<500 500 500>
    componentLinearAccScale=<0.2 0.2 0.2>
    componentLinearVelScale=<0.2 0.2 0.2>
    gravityOverride {z=-900}
    physicsBodyDefinitions [{
      boundBone {boneName=dyn_scarf_04}
      boxExtents=<13 10 8>
      constraintSetup {angularLimitsMax=<0 0 30>, angularLimitsMin=<0 -45 -30>, angularTarget=<0 0 0>}
      localJointOffset=<1.5 0 0>
     }]
   }
   Output Pose(Result=Blend Poses by bool(BlendPose_0=Component To Local(ComponentPose=@n2), BlendPose_1=Component To Local(ComponentPose=@n10), bActiveValue=Get bRigidBodyEnabled()) {blendType=Linear}) {layerGroup=None}
   n11 = Control Rig!(Source=Copy Pose From Mesh() {bUseAttachedParent=true}, Alpha, bIsInAirOrEmote:bool=Get bIsInAirOrEmote() -> (bIsInAirOrEmote:bool)) -> (Pose) {
    alphaBoolBlend {blendOption=Linear}
    controlRigAssetReference {blueprintRigClass=&/F_MED_BuffBeak_CtrlRig.F_MED_BuffBeak_CtrlRig_C}
   }
   n12 = "Transform (Modify) Bone - Bone: clavicle_r"!(ComponentPose=Local To Component(LocalPose=@n11), bAlphaBoolEnabled:bool=True, Translation="0,0,0", Rotation="0,0,20", Scale="1,1,1") -> (Pose) {
    alphaInputType=Bool
    rotationMode=BMM_Replace
    rotationSpace=BCS_ParentBoneSpace
    translationMode=BMM_Replace
    translationSpace=BCS_ParentBoneSpace
    alphaBoolBlend {blendOption=Linear}
    boneToModify {boneName=clavicle_r}
    rotation {roll=20}
   }
   n13 = "Transform (Modify) Bone - Bone: upperarm_r"!(ComponentPose=@n12, bAlphaBoolEnabled:bool=True, Translation="0,0,0", Rotation="0,0,20", Scale="1,1,1") -> (Pose) {
    alphaInputType=Bool
    rotationMode=BMM_Replace
    rotationSpace=BCS_ParentBoneSpace
    translationMode=BMM_Replace
    translationSpace=BCS_ParentBoneSpace
    alphaBoolBlend {blendOption=Linear}
    boneToModify {boneName=upperarm_r}
    rotation {roll=20}
   }
   n14 = "Transform (Modify) Bone - Bone: lowerarm_r"!(ComponentPose=@n13, bAlphaBoolEnabled:bool=True, Translation="0,0,0", Rotation="0,0,20", Scale="1,1,1") -> (Pose) {
    alphaInputType=Bool
    rotationMode=BMM_Replace
    rotationSpace=BCS_ParentBoneSpace
    translationMode=BMM_Replace
    translationSpace=BCS_ParentBoneSpace
    alphaBoolBlend {blendOption=Linear}
    boneToModify {boneName=lowerarm_r}
    rotation {roll=20}
   }
   n15 = "Transform (Modify) Bone - Bone: hand_r"!(ComponentPose=@n14, bAlphaBoolEnabled:bool=True, Translation="0,0,0", Rotation="0,0,20", Scale="1,1,1") -> (Pose) {
    alphaInputType=Bool
    rotationMode=BMM_Replace
    rotationSpace=BCS_ParentBoneSpace
    translationMode=BMM_Replace
    translationSpace=BCS_ParentBoneSpace
    alphaBoolBlend {blendOption=Linear}
    boneToModify {boneName=hand_r}
    rotation {roll=20}
   }
   Save cached pose 'PreDynamics'(Pose=Component To Local(ComponentPose=@n15)) {cacheName=PreDynamics}
  ]}
 EventGraph {nodes [
   Event BlueprintUpdateAnimation()
   > n1 = Sequence()
   > n2 = Gravity Override State Function DEPRECATED(
    CharacterPartAnimInstance=Self-Reference()
    GameTimeInSeconds=GetGameTimeInSeconds()
    GravityOverrideParams=Get Gravity Override Params()
    RBAN Params=Get RBAN Params()
   )
   > Set Gravity Override(Gravity Override=@n2)
   > n3 = Set bIsBodyValid(bIsBodyValid:bool=IsValid(Object=Get BodyMesh() -> (BodyMesh:Skeletal Mesh Component Object Reference))) -> (Output_Get:bool)
   > Branch(Condition=@n3)
   n4 = BreakTransform(InTransform=GetSocketTransform(self=Get BodyMesh(), InSocketName=clavicle_r, TransformSpace=RTS_ParentBoneSpace))
   > Set clavicle_r_Location(clavicle_r_Location:Vector=@n4) -> (Output_Get:Vector)
   > Set clavicle_r_Rotation(clavicle_r_Rotation:Rotator=@n4.Rotation) -> (Output_Get:Rotator)
   n5 = BreakTransform(InTransform=GetSocketTransform(self=Get BodyMesh(), InSocketName=upperarm_r, TransformSpace=RTS_ParentBoneSpace))
   > Set upperarm_r_Location(upperarm_r_Location:Vector=@n5) -> (Output_Get:Vector)
   > Set upperarm_r_Rotation(upperarm_r_Rotation:Rotator=@n5.Rotation) -> (Output_Get:Rotator)
   n6 = BreakTransform(InTransform=GetSocketTransform(self=Get BodyMesh(), InSocketName=lowerarm_r, TransformSpace=RTS_ParentBoneSpace))
   > Set lowerarm_r_Location(lowerarm_r_Location:Vector=@n6) -> (Output_Get:Vector)
   > Set lowerarm_r_Rotation(lowerarm_r_Rotation:Rotator=@n6.Rotation) -> (Output_Get:Rotator)
   n7 = BreakTransform(InTransform=GetSocketTransform(self=Get BodyMesh(), InSocketName=hand_r, TransformSpace=RTS_ParentBoneSpace))
   > Set Hand_r_Location(Hand_r_Location:Vector=@n7) -> (Output_Get:Vector)
   > Set Hand_r_Rotation(Hand_r_Rotation:Rotator=@n7.Rotation) -> (Output_Get:Rotator)
   Set bIsInAirOrEmote(execute=@n1.then_1, bIsInAirOrEmote:bool=OR Boolean(A=Get bIsPlayingEmote(), B=Get bIsSkyDiving())) -> (Output_Get:bool)
   n8 = Event OnCharacterCustomizationLoaded()
   > Set BodyMesh(BodyMesh=Get Array Item with EFort Custom Part Type(Array=@n8.CustomMeshParts))
  ]}
]
variables [
 bIsBodyValid {type=boolean}
 bIsInAirOrEmote {type=boolean}
 BodyMesh {type=SkeletalMeshComponent}
 clavicle_r_Location {type=Vector}
 clavicle_r_Rotation {type=Rotator}
 Gravity Override {type=Vector}
 Gravity Override Params {type=GravityOverrideParamsStruct}
 Hand_r_Location {type=Vector}
 Hand_r_Rotation {type=Rotator}
 lowerarm_r_Location {type=Vector}
 lowerarm_r_Rotation {type=Rotator}
 RBAN Params {
  type=RBANParamsStruct
  default {
   dBNO {simSpaceSettings {
     maxAngularAcceleration=1000
     maxAngularVelocity=3
     maxLinearAcceleration=300
     maxLinearVelocity=700
     velocityScaleZ=0
     externalLinearDragV=<0 0.5 0>
    }}
   dBNO_Carried {simSpaceSettings {
     maxAngularAcceleration=1000
     maxAngularVelocity=3
     maxLinearAcceleration=300
     maxLinearVelocity=700
     velocityScaleZ=0
     externalLinearDragV=<0 0.5 0>
    }}
   emoteOrMelee {simSpaceSettings {
     maxAngularAcceleration=1000
     maxAngularVelocity=3
     maxLinearAcceleration=300
     maxLinearVelocity=700
     velocityScaleZ=0
     externalLinearDragV=<0 0.5 0>
    }}
   front_End {simSpaceSettings {
     maxAngularAcceleration=1000
     maxAngularVelocity=3
     maxLinearAcceleration=300
     maxLinearVelocity=700
     velocityScaleZ=0
     externalLinearDragV=<0 0.5 0>
    }}
   onGround_Crouch_Moving {simSpaceSettings {
     maxAngularAcceleration=1000
     maxAngularVelocity=3
     maxLinearAcceleration=300
     maxLinearVelocity=700
     velocityScaleZ=0
     externalLinearDragV=<0 0.5 0>
    }}
   onGround_Falling {simSpaceSettings {
     maxAngularAcceleration=1000
     maxAngularVelocity=3
     maxLinearAcceleration=300
     maxLinearVelocity=700
     velocityScaleZ=0
     externalLinearDragV=<0 0.5 0>
    }}
   onGround_Moving {simSpaceSettings {
     maxAngularAcceleration=1000
     maxAngularVelocity=3
     maxLinearAcceleration=300
     maxLinearVelocity=700
     velocityScaleZ=0
     externalLinearDragV=<0 0.5 0>
    }}
   onGround_Standing {simSpaceSettings {
     maxAngularAcceleration=1000
     maxAngularVelocity=3
     maxLinearAcceleration=300
     maxLinearVelocity=700
     velocityScaleZ=0
     externalLinearDragV=<0 0.5 0>
    }}
   rail_Grinding {simSpaceSettings {
     maxAngularAcceleration=1000
     maxAngularVelocity=3
     maxLinearAcceleration=300
     maxLinearVelocity=700
     velocityScaleZ=0
     worldAlpha=0.1
     externalLinearDragV=<0 0.5 0>
    }}
   riding_Moving {simSpaceSettings {
     maxAngularAcceleration=1000
     maxAngularVelocity=3
     maxLinearAcceleration=300
     maxLinearVelocity=700
     velocityScaleZ=0
     externalLinearDragV=<0 0.5 0>
    }}
   riding_Moving_Targeting {simSpaceSettings {
     maxAngularAcceleration=1000
     maxAngularVelocity=3
     maxLinearAcceleration=300
     maxLinearVelocity=700
     velocityScaleZ=0
     externalLinearDragV=<0 0.5 0>
    }}
   riding_Standing {simSpaceSettings {
     maxAngularAcceleration=1000
     maxAngularVelocity=3
     maxLinearAcceleration=300
     maxLinearVelocity=700
     velocityScaleZ=0
     externalLinearDragV=<0 0.5 0>
    }}
   skydive_DivingDown {simSpaceSettings {
     maxAngularAcceleration=1000
     maxAngularVelocity=3
     maxLinearAcceleration=300
     maxLinearVelocity=700
     velocityScaleZ=0
     externalLinearDragV=<0 0.5 0>
    }}
   skydive_FloatingDown {
    componentLinearAccScale=<0 0.5 0>
    componentLinearVelScale=<0 0.5 0>
    simSpaceSettings {
     maxAngularAcceleration=1000
     maxAngularVelocity=3
     maxLinearAcceleration=300
     maxLinearVelocity=700
     velocityScaleZ=0
     externalLinearDragV=<0 0.5 0>
    }
   }
   skydive_GoingUp {simSpaceSettings {
     maxAngularAcceleration=1000
     maxAngularVelocity=3
     maxLinearAcceleration=300
     maxLinearVelocity=700
     velocityScaleZ=0
     externalLinearDragV=<0 0.5 0>
    }}
   skydive_Parachuting {
    componentLinearAccScale=<0 0.1 0.3>
    componentLinearVelScale=<0 0.1 0.3>
    simSpaceSettings {
     maxAngularAcceleration=1000
     maxAngularVelocity=3
     maxLinearAcceleration=300
     maxLinearVelocity=700
     velocityScaleZ=0
     externalLinearDragV=<0 0.5 0>
    }
   }
   skydive_Surfing {simSpaceSettings {
     maxAngularAcceleration=1000
     maxAngularVelocity=3
     maxLinearAcceleration=300
     maxLinearVelocity=700
     velocityScaleZ=0
     externalLinearDragV=<0 0.5 0>
    }}
   swimming {simSpaceSettings {
     maxAngularAcceleration=1000
     maxAngularVelocity=3
     maxLinearAcceleration=300
     maxLinearVelocity=700
     velocityScaleZ=0
     externalLinearDragV=<0 0.5 0>
    }}
   tactical_Sprint {simSpaceSettings {
     maxAngularAcceleration=1000
     maxAngularVelocity=3
     maxLinearAcceleration=300
     maxLinearVelocity=700
     velocityScaleZ=0
     externalLinearDragV=<0 0.5 0>
    }}
   ziplining {simSpaceSettings {
     maxAngularAcceleration=1000
     maxAngularVelocity=3
     maxLinearAcceleration=300
     maxLinearVelocity=700
     velocityScaleZ=0
     externalLinearDragV=<0 0.5 0>
    }}
  }
 }
 upperarm_r_Location {type=Vector}
 upperarm_r_Rotation {type=Rotator}
]
